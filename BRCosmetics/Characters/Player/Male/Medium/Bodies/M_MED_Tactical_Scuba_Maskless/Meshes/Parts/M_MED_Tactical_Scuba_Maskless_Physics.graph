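# Physics asset for the maskless Tactical Scuba male medium body, covering upper spine, neck, head and dynamic hair and facial-hair bones.
# Capsules and a sphere shape the spine, neck and head chain, linked by angularly free constraints.
# Bangs, side and back hair, beard and mustache form short chains off the head with tight swing and twist limits.
# The lock and the beard and mustache base bodies attach to the head with free rotation.
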PhysicsAsset /BRCosmetics/Characters/Player/Male/Medium/Bodies/M_MED_Tactical_Scuba_Maskless/Meshes/Parts/M_MED_Tactical_Scuba_Maskless_Physics
bodies {
 dyn_bang_01_l [{shapeType=Capsule, length=3.3444, radius=1.1926, center=<1.6657 0.1438 0.0001>, rotation=<rot 90 0 0>}]
 dyn_bang_01_r [{shapeType=Capsule, length=3.2066, radius=0.8785, center=<1.3895 0 0>, rotation=<rot 90 0 0>}]
 dyn_bang_02_l [{shapeType=Capsule, length=3.0285, radius=1.0566, center=<1.3895 0 0>, rotation=<rot 90 0 0>}]
 dyn_bang_02_r [{shapeType=Capsule, length=6.8212, radius=0.8785, center=<3.4633 0 0.0001>, rotation=<rot 90 0 0>}]
 dyn_bang_03_l [{shapeType=Capsule, length=3.2066, radius=0.8785, center=<1.3895 0 0>, rotation=<rot 90 0 0>}]
 dyn_bang_03_r [{shapeType=Capsule, length=2.7759, radius=0.7536, center=<1.3895 0 0>, rotation=<rot 90 0 0>}]
 dyn_bang_04_r [{shapeType=Capsule, length=1.6659, radius=0.6186, center=<1.0692 0 0>, rotation=<rot 90 0 0>}]
 dyn_beard_01 [{shapeType=Capsule, length=6.4958, radius=3.8465, center=<1.9668 0.8385 0>}]
 dyn_beard_01_Base [{shapeType=Box, extentX=1.01, extentY=1.01, extentZ=1.01, rotation=<rot 0 -90 0>}]
 dyn_beard_02 [{shapeType=Capsule, length=7.3188, radius=2.6601, center=<1.6415 -0.3836 -0.0934>, rotation=<rot 0 0 -13.6817>}]
 dyn_beard_03 [{shapeType=Capsule, length=2.5637, radius=3.0668, center=<0.9212 -1.1865 0.0001>}]
 dyn_hairBack_01 [{shapeType=Box, extentX=3.6245, extentY=1.2265, extentZ=10.8125, center=<1.5 0 0>, rotation=<rot 0 0 8.9843>}]
 dyn_hairBack_01_Base [{shapeType=Box, extentX=1.01, extentY=1.01, extentZ=1.01, rotation=<rot 0 -90 0>}]
 dyn_hairBack_02 [{shapeType=Box, extentX=3.9587, extentY=1.2265, extentZ=10.8125, center=<0.9851 -0.0622 -0.394>, rotation=<rot 13.444 2.1662 9.2414>}]
 dyn_hairBack_03 [{shapeType=Box, extentX=5.4397, extentY=1.2265, extentZ=10.8125, center=<2.2403 0 0>, rotation=<rot 0 0 8.9847>}]
 dyn_hairSide_01_l [{shapeType=Box, extentX=3.6245, extentY=1.2265, extentZ=10.8125, center=<1.5 0 0>, rotation=<rot 0 0 8.9843>}]
 dyn_hairSide_01_r [{shapeType=Box, extentX=3.6245, extentY=1.2265, extentZ=10.8125, center=<1.5 0 0>, rotation=<rot 0 0 8.9843>}]
 dyn_hairSide_02_l [{shapeType=Box, extentX=3.9587, extentY=1.2265, extentZ=10.8125, center=<1.5 0 0>, rotation=<rot 0 0 8.9843>}]
 dyn_hairSide_02_r [{shapeType=Box, extentX=3.9587, extentY=1.2265, extentZ=10.8125, center=<1.5 0 0>, rotation=<rot 0 0 8.9843>}]
 dyn_hairSide_03_l [{shapeType=Box, extentX=5.4397, extentY=1.2265, extentZ=10.8125, center=<2.2403 0 0>, rotation=<rot 0 0 8.9847>}]
 dyn_hairSide_03_r [{shapeType=Box, extentX=5.4397, extentY=1.2265, extentZ=10.8125, center=<2.2403 0 0>, rotation=<rot 0 0 8.9847>}]
 dyn_lockA_l [{shapeType=Capsule, length=2.3961, radius=0.8801, center=<1.3081 0 0>, rotation=<rot 90 0 0>}]
 dyn_mustache_01_Base_l [{shapeType=Capsule, length=0.505, radius=0.505, rotation=<rot 0 -90 0>}]
 dyn_mustache_01_Base_r [{shapeType=Capsule, length=0.505, radius=0.505, rotation=<rot 0 -90 0>}]
 dyn_mustache_01_l [{shapeType=Capsule, length=2.7121, radius=1.2763, center=<2 0 0>, rotation=<rot 90 0 0>}]
 dyn_mustache_01_r [{shapeType=Capsule, length=3.0641, radius=1.4337, center=<2 0 0>, rotation=<rot 90 0 0>}]
 dyn_mustache_02_l [{shapeType=Capsule, length=2.7121, radius=1.2763, center=<2 0 0>, rotation=<rot 90 0 0>}]
 dyn_mustache_02_r [{shapeType=Capsule, length=2.7121, radius=1.2763, center=<2 0 0>, rotation=<rot 90 0 0>}]
 head [{shapeType=Sphere, radius=7.6848}, {shapeType=Capsule, length=13.5902, radius=7.117, center=<0 1.9434 0.0001>, rotation=<rot 90 0 0>}]
 neck_01 [
  {shapeType=Capsule, length=0, radius=8.7181, center=<0 -0.8532 0>, rotation=<rot 0 -90 0>}
  {shapeType=Capsule, length=17.0744, radius=1.7217, center=<3.8164 0.6208 8.9322>, rotation=<rot 0 21.2516 89.9999>}
 ]
 neck_02 [{shapeType=Capsule, length=0.2299, radius=10.3501, center=<-3.6387 -1.646 0>, rotation=<rot 90 9.4623 -9.1872>}]
 spine_03 [{shapeType=Capsule, length=5.0864, radius=11.3579, center=<11.8492 0 0>, rotation=<rot 0 -90 0>}]
 spine_04 [{shapeType=Capsule, length=8.7858, radius=11.3579, center=<6.7134 0 0>, rotation=<rot 0 -90 0>}]
 spine_05 [{shapeType=Capsule, length=12.6868, radius=11.3579, center=<1.628 0 0>, rotation=<rot 0 -90 0>}]
}
constraints [
 {bone1Name=dyn_bang_01_r, bone2Name=head, swing1LimitDegrees=6, swing2LimitDegrees=6, twistLimitDegrees=6}
 {bone1Name=dyn_bang_02_r, bone2Name=dyn_bang_01_r, swing1LimitDegrees=6, swing2LimitDegrees=6, twistLimitDegrees=6}
 {bone1Name=dyn_bang_03_r, bone2Name=dyn_bang_02_r, swing1LimitDegrees=6, swing2LimitDegrees=6, twistLimitDegrees=6}
 {bone1Name=dyn_bang_04_r, bone2Name=dyn_bang_03_r, swing1LimitDegrees=6, swing2LimitDegrees=6, twistLimitDegrees=6}
 {bone1Name=dyn_bang_01_l, bone2Name=head, swing1LimitDegrees=6, swing2LimitDegrees=6, twistLimitDegrees=6}
 {bone1Name=dyn_bang_02_l, bone2Name=dyn_bang_01_l, swing1LimitDegrees=12, swing2LimitDegrees=6, twistLimitDegrees=6}
 {bone1Name=dyn_bang_03_l, bone2Name=dyn_bang_02_l, swing1LimitDegrees=12, swing2LimitDegrees=6, twistLimitDegrees=6}
 {
  bone1Name=dyn_lockA_l
  bone2Name=head
  swing1LimitDegrees=45
  swing1Motion=Free
  swing2LimitDegrees=45
  swing2Motion=Free
  twistLimitDegrees=45
  twistMotion=Free
 }
 {bone1Name=dyn_hairSide_01_l, bone2Name=head, swing1LimitDegrees=8, swing2LimitDegrees=8, twistLimitDegrees=8}
 {bone1Name=dyn_hairSide_02_l, bone2Name=dyn_hairSide_01_l, swing1LimitDegrees=8, swing2LimitDegrees=8, twistLimitDegrees=8}
 {bone1Name=dyn_hairSide_03_l, bone2Name=dyn_hairSide_02_l, swing1LimitDegrees=8, swing2LimitDegrees=8, twistLimitDegrees=8}
 {bone1Name=dyn_hairSide_01_r, bone2Name=head, swing1LimitDegrees=8, swing2LimitDegrees=8, twistLimitDegrees=8}
 {bone1Name=dyn_hairSide_02_r, bone2Name=dyn_hairSide_01_r, swing1LimitDegrees=8, swing2LimitDegrees=8, twistLimitDegrees=8}
 {bone1Name=dyn_hairSide_03_r, bone2Name=dyn_hairSide_02_r, swing1LimitDegrees=8, swing2LimitDegrees=8, twistLimitDegrees=8}
 {bone1Name=dyn_hairBack_01_Base, bone2Name=head, swing1LimitDegrees=8, swing2LimitDegrees=8, twistLimitDegrees=8}
 {bone1Name=dyn_hairBack_01, bone2Name=dyn_hairBack_01_Base, swing1LimitDegrees=14, swing2LimitDegrees=14, twistLimitDegrees=8}
 {bone1Name=dyn_hairBack_02, bone2Name=dyn_hairBack_01, swing1LimitDegrees=6, swing2LimitDegrees=12, twistLimitDegrees=8}
 {bone1Name=dyn_hairBack_03, bone2Name=dyn_hairBack_02, swing1LimitDegrees=6, swing2LimitDegrees=12, twistLimitDegrees=8}
 {
  bone1Name=dyn_beard_01_Base
  bone2Name=head
  swing1LimitDegrees=45
  swing1Motion=Free
  swing2LimitDegrees=45
  swing2Motion=Free
  twistLimitDegrees=45
  twistMotion=Free
 }
 {bone1Name=dyn_beard_01, bone2Name=dyn_beard_01_Base, swing1LimitDegrees=20, swing2LimitDegrees=12, twistLimitDegrees=4}
 {bone1Name=dyn_beard_02, bone2Name=dyn_beard_01, swing1LimitDegrees=18, swing2LimitDegrees=14, twistLimitDegrees=4}
 {bone1Name=dyn_beard_03, bone2Name=dyn_beard_02, swing1LimitDegrees=18, swing2LimitDegrees=14, twistLimitDegrees=4}
 {
  bone1Name=head
  bone2Name=neck_02
  swing1LimitDegrees=45
  swing1Motion=Free
  swing2LimitDegrees=45
  swing2Motion=Free
  twistLimitDegrees=45
  twistMotion=Free
 }
 {
  bone1Name=spine_05
  bone2Name=spine_04
  swing1LimitDegrees=45
  swing1Motion=Free
  swing2LimitDegrees=45
  swing2Motion=Free
  twistLimitDegrees=45
  twistMotion=Free
 }
 {
  bone1Name=spine_04
  bone2Name=spine_03
  swing1LimitDegrees=45
  swing1Motion=Free
  swing2LimitDegrees=45
  swing2Motion=Free
  twistLimitDegrees=45
  twistMotion=Free
 }
 {
  bone1Name=dyn_mustache_01_Base_r
  bone2Name=head
  swing1LimitDegrees=45
  swing1Motion=Free
  swing2LimitDegrees=45
  swing2Motion=Free
  twistLimitDegrees=45
  twistMotion=Free
 }
 {bone1Name=dyn_mustache_01_r, bone2Name=dyn_mustache_01_Base_r, swing1LimitDegrees=6, swing2LimitDegrees=6, twistLimitDegrees=6}
 {bone1Name=dyn_mustache_02_r, bone2Name=dyn_mustache_01_r, swing1LimitDegrees=6, swing2LimitDegrees=6, twistLimitDegrees=6}
 {
  bone1Name=dyn_mustache_01_Base_l
  bone2Name=head
  swing1LimitDegrees=45
  swing1Motion=Free
  swing2LimitDegrees=45
  swing2Motion=Free
  twistLimitDegrees=45
  twistMotion=Free
 }
 {bone1Name=dyn_mustache_01_l, bone2Name=dyn_mustache_01_Base_l, swing1LimitDegrees=6, swing2LimitDegrees=6, twistLimitDegrees=6}
 {bone1Name=dyn_mustache_02_l, bone2Name=dyn_mustache_01_l, swing1LimitDegrees=6, swing2LimitDegrees=6, twistLimitDegrees=6}
 {
  bone1Name=neck_01
  bone2Name=spine_05
  swing1LimitDegrees=45
  swing1Motion=Free
  swing2LimitDegrees=45
  swing2Motion=Free
  twistLimitDegrees=45
  twistMotion=Free
 }
 {
  bone1Name=neck_02
  bone2Name=neck_01
  swing1LimitDegrees=45
  swing1Motion=Free
  swing2LimitDegrees=45
  swing2Motion=Free
  twistLimitDegrees=45
  twistMotion=Free
 }
]
